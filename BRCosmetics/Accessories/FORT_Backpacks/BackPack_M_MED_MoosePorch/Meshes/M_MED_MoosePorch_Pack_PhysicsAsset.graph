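# Physics asset for the MoosePorch medium male backpack cosmetic mesh.
# It defines a single capsule body on the pack bone, tilted and offset to wrap the bag's volume.
# It has no constraints, so the pack has no jointed or dangling simulated parts.
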PhysicsAsset /BRCosmetics/Accessories/FORT_Backpacks/BackPack_M_MED_MoosePorch/Meshes/M_MED_MoosePorch_Pack_PhysicsAsset
bodies {pack [{shapeType=Capsule, length=22.8156, radius=11.044, center=<-0.0017 -10.5957 -4.1074>, rotation=<rot -0.0389 -90 0.0127>}]}
constraints []
